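# Niagara deploy effect for the male Chicken Warrior glider, bursting feathers and shell fragments.
# Idle_Feathers bursts mesh and sprite particles from User.SkeletalMesh with inherited velocity, curve-ramped gravity and drag, spin, and camera-based shrinking; its rate spawners are disabled.
# Shell_Bits is a GPU mesh burst above the origin, where one Distribution value drives size, launch speed and fall, so bigger bits launch slower.
# User.VariantNum switches both emitters' colors.
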NiagaraSystem /BRCosmetics/Gadgets/Assets/VinderTech_GliderChute/Glider_ChickenWarrior_Male/FX/NS_Glider_ChickenWarrior_Deploy
dependencies {
 usedDataInterfaces [&NiagaraDataInterfaceCurve, &NiagaraDataInterfaceArrayFloat, &NiagaraDataInterfaceCamera]
 usedDynamicInputs []
 usedModules [&SystemState]
}
emitters {
 Idle_Feathers {
  emitterUpdateScript [
   &EmitterState()
   &SpawnRate(
    SpawnGroup:NiagaraInt32=&RandomRangeInt(Minimum=0, Maximum=4, Evaluation Type=::NewEnumerator1)
    SpawnRate=&Multiply_Float(
     A=&Lerp_Float(
      A=0
      B=8
      Alpha=&Power(a=&RandomRangeFloat(Minimum=0, Maximum=1, Evaluation Type=::NewEnumerator1), b=10)
     )
     B=$System.IdleSpawnRate
    )
    Use Spawn Probability=true
   ) {enabled=false}
   &SpawnRate(
    SpawnGroup:NiagaraInt32=&RandomRangeInt(Minimum=0, Maximum=8, Evaluation Type=::NewEnumerator1)
    SpawnRate=500
    Use Spawn Probability=false
   ) {enabled=false, module=SpawnRate001}
   &SpawnBurst_Instantaneous(Spawn Count=20, Use Spawn Probability=false)
  ]
  particleSpawnScript [
   /Niagara/Modules/Spawn/Initialization/InitializeParticle(
    Color=&LerpLinearColors(
     StartColor=<rgba 1 0.9207 0 1>
     EndColor=<rgba 0.0417 0.0417 0.0417 1>
     LerpFactor=&MakeFloatFromInt(Int=$User.VariantNum)
    )
    Lifetime=&RandomRangeFloat(Minimum=1, Maximum=2)
    Mesh Scale=&Multiply_VectorByFloat(Vector=<1 1 1>, Float=1)
    Module.Write Color=true
    Module.Write Lifetime=true
    Module.Write Mass=true
    Module.Write Material Random=true
    Module.Write Position=false
    Module.Write Scale=true
    Module.Write SpriteRotation=false
    Module.Write SpriteSize=true
    Position=&SimulationPosition()
    Sprite Size=<1 1>
   )
   &SkeletalMeshLocation(
    Apply Sampled Orientation to Mesh Particles=true
    Bone Sampling Mode=::NewEnumerator0
    Inherit Velocity=false
    MaskSpawnGroup=false
    Module.Override Randomness=false
    Skeletal Mesh=$User.SkeletalMesh
   )
   &DynamicMaterialParameters(
    Index 0 Param 1=&SelectFloatFromArray(
     Direct Array Index=&SpawnGroup()
     Float Selection Array=DataInterface() {propertyValues {FloatData [0,0.75,0.25,0.75,0.75,0.5,0.75]}}
    )
    Index 0 Param 2=&FloatFromCurve(FloatCurve=DataInterface() {propertyValues {Curve {keys [<at 0 0>,<at 0.5 0.25>]}}})
    Index 0 Param 3=&FloatFromCurve(FloatCurve=DataInterface() {propertyValues {Curve {keys [<at 0 1>,<at 0.1 0.5>,<at 1 0>]}}})
    Index 0 Param 4=&FloatFromCurve(FloatCurve=DataInterface() {propertyValues {Curve {keys [<at 0 1>,<at 1 0>]}}})
    Module.Param0WriteEnabled=true
    Module.Param1WriteEnabled=false
    Module.Param2WriteEnabled=false
    Module.Param3WriteEnabled=false
   ) {module=DynamicMaterialParameters001}
   &AddVelocity(
    Coordinate Space=::Simulation
    Velocity=&RotateVectorByQuaternion(Vector To Rotate=<-50 0 0>, Quaternion=$Output.SkeletalMeshLocation.SampledRotation)
   ) {enabled=false}
   /SetVariables(
    Particles.NewQuat:Quat4f=$Output.SkeletalMeshLocation.SampledRotation
    Particles.NewVector:Vector3f=$Output.SkeletalMeshLocation.SampledPosition
    Particles.SpinDirection:NiagaraFloat=&UniformAOrBFloat(A=-1)
   ) {isSetParameters=true}
   &AddVelocityInCone(
    Cone Angle=65
    Cone Axis=&RandomRangeVector(Minimum=<-1 0 0>, Maximum=<1 0 0>)
    Module.Use Velocity Falloff On Cone Axis=false
    Velocity Strength=/Game/Effects/Niagara/DynamicInputs/Math/FN_Fit01(
     Clamp:NiagaraBool=false
     Value:NiagaraFloat=&OneMinusFloat(Float=0)
     DestMin:NiagaraFloat=50
     DestMax:NiagaraFloat=900
    )
   )
   &InheritVelocity(Module.Limit Speed=true, Velocity Limit=300, Velocity Scale=&Multiply_VectorByFloat(Vector=<1 1 1>, Float=20))
   &AddVelocity(Velocity=&RandomRangeVector(Minimum=<0 -15 0>, Maximum=<1 -20 1>)) {enabled=false, module=AddVelocity001}
   &Color(Color=<rgba 1 0.7922 0.0157 1>) {enabled=false}
  ]
  particleUpdateScript [
   &self:CameraSize_02(SomeValueDivisor=50, ZeroOffset=-75) {module=CameraSize_02}
   &ParticleState()
   &GravityForce(Gravity=&Multiply_VectorByFloat(
     Vector=<1 1 -250>
     Float=&FloatFromCurve(FloatCurve=DataInterface() {propertyValues {Curve {keys [{arriveTangent=2.0378, interpMode=RCIM_Cubic, leaveTangent=2.0378, tangentMode=RCTM_User, time=0, value=0}, <at 1 1>]}}})
    ))
   &Drag(
    Drag=&Multiply_Float(
     A=6
     B=&FloatFromCurve(FloatCurve=DataInterface() {propertyValues {Curve {keys [
          {arriveTangent=2.4781, interpMode=RCIM_Cubic, leaveTangent=2.4781, tangentMode=RCTM_User, time=0, value=0}
          {arriveTangent=0.1523, interpMode=RCIM_Cubic, leaveTangent=0.1523, tangentMode=RCTM_User, time=1, value=1}
         ]}}})
    )
    Use Linear Drag=true
    Use Rotational Drag=false
   )
   &AddVelocity(Velocity=<0 -1 0>) {module=AddVelocity002}
   &SolveForcesAndVelocity(Clamp Velocity=false, Limit Acceleration=false)
   &DynamicMaterialParameters(
    Index 0 Param 1=0.75
    Index 0 Param 2=&FloatFromCurve(FloatCurve=DataInterface() {propertyValues {Curve {keys [<at 0 0>,<at 0.5 0.25>]}}})
    Index 0 Param 3=&FloatFromCurve(FloatCurve=DataInterface() {propertyValues {Curve {keys [<at 0 1>,<at 0.1 0.5>,<at 1 0>]}}})
    Index 0 Param 4=&FloatFromCurve(FloatCurve=DataInterface() {propertyValues {Curve {keys [
         <at 0 0.5 RCIM_Cubic>
         {arriveTangent=0.7684, interpMode=RCIM_Cubic, leaveTangent=0.7684, time=0.26, value=0.5807}
         {arriveTangent=-0.7847, interpMode=RCIM_Cubic, leaveTangent=-0.7847, time=0.5225, value=0.9015}
         <at 1 0>
        ]}}})
    Module.Param0WriteEnabled=false
    Module.Param1WriteEnabled=true
    Module.Param2WriteEnabled=true
    Module.Param3WriteEnabled=true
   ) {enabled=false}
   &UpdateMeshOrientation(
    Rotation Coordinate Space=::NewEnumerator3
    Rotation Rate=&Multiply_Float(
     A=&Multiply_Float(A=&RandomRangeFloat(Minimum=1.5, Maximum=2), B=$Particles.SpinDirection)
     B=&FloatFromCurve(FloatCurve=DataInterface() {propertyValues {Curve {keys [
          {arriveTangent=0.0162, interpMode=RCIM_Cubic, leaveTangent=0.0162, tangentMode=RCTM_User, time=0, value=0}
          {arriveTangent=0.0117, interpMode=RCIM_Cubic, leaveTangent=0.0038, tangentMode=RCTM_Break, time=0.25, value=0.9425}
          <at 1 0.5>
         ]}}})
    )
    Rotation Vector=&Add_Vector(A=<1 1 1>, B=&RandomRangeVector(Minimum=<0 0 0>, Maximum=<0 0 0>))
   )
   &ScaleMeshSize(Scale Factor=&Multiply_Vector(
     A=&VectorFromFloat(Value=&FloatFromCurve(FloatCurve=DataInterface() {propertyValues {Curve {keys [<at 0.75 1 RCIM_Cubic>, {arriveTangent=-10.6879, interpMode=RCIM_Cubic, leaveTangent=-10.6879, tangentMode=RCTM_User, time=1, value=0}]}}}))
     B=&Lerp_Vector(A=<0 0 0>, B=<1 1 1>, Alpha=&ClampFloat(Float=$Particles.CameraShrink))
    ))
  ]
  rendererClasses [&NiagaraMeshRendererProperties, &NiagaraSpriteRendererProperties]
  renderers [
   {rendererClass=&NiagaraMeshRendererProperties, rendererIndex=0}
   {rendererClass=&NiagaraSpriteRendererProperties, rendererIndex=1}
  ]
 }
 Shell_Bits {
  simTarget=GPUComputeSim
  emitterUpdateScript [
   &EmitterState(Enable Distance Culling=true, Module.bUseMaxDistance=true, Module.bUseMinDistance=false, Scalability Mode=::NewEnumerator1)
   &SpawnBurst_Instantaneous(
    Spawn Count=&RandomRangeInt(Minimum=10, Maximum=20, Evaluation Type=::NewEnumerator1)
    Use Spawn Probability=false
   )
  ]
  particleSpawnScript [
   /SetVariables(Particles.Distribution:NiagaraFloat=&Add_Float(
     A=&UniformAOrBFloat(A=0.25, B=0.9, Distribution Weight=0.1)
     B=&RandomRangeFloat(Minimum=-0.1, Maximum=0.1)
    )) {isSetParameters=true}
   &InitializeParticle(
    Color=&LerpLinearColors(
     StartColor=&LerpLinearColors(
      StartColor=<rgba 0.1719 0.0451 0.1584 1>
      EndColor=<rgba 0.5885 0.2299 0.5262 1>
      LerpFactor=/Niagara/DynamicInputs/Integer/V2/Round_Float(Float:NiagaraFloat=&RandomRangeFloat(Minimum=0, Maximum=1))
     )
     EndColor=&RandomRangeLinearColor(Minimum=<rgba 0.0104 0.0104 0.0104 1>, Maximum=<rgba 0.224 0.224 0.224 1>)
     LerpFactor=&MakeFloatFromInt(Int=$User.VariantNum)
    )
    Lifetime=&UniformRangedFloat(Minimum=0.4, Maximum=0.8)
    Mesh Scale=&VectorFromFloat(Value=&FN_Fit01(Clamp:NiagaraBool=false, Value:NiagaraFloat=$Particles.Distribution, DestMin:NiagaraFloat=0.02, DestMax:NiagaraFloat=0.08))
    Module.Write Color=true
    Module.Write Lifetime=true
    Module.Write Mass=true
    Module.Write Material Random=true
    Module.Write Position=false
    Module.Write Scale=true
    Module.Write SpriteRotation=false
    Module.Write SpriteSize=false
    Position=&SimulationPosition()
    Sprite Size=&UniformRangedVector2D(Minimum=<2.1064 4.7118>, Maximum=<4.4317 8.5012>)
   )
   &InitialMeshOrientation(Orientation Vector=&RandomVector(Vector Scale=1), Use Orientation Vector=true, Use Rotation Vector=false)
   &AddVelocityInCone(
    Cone Angle=65
    Cone Axis=&RandomRangeVector(Minimum=<-1 0 0>, Maximum=<1 0 0>)
    Module.Use Velocity Falloff On Cone Axis=false
    Velocity Strength=&FN_Fit01(
     Clamp:NiagaraBool=false
     Value:NiagaraFloat=&OneMinusFloat(Float=$Particles.Distribution)
     DestMin:NiagaraFloat=300
     DestMax:NiagaraFloat=900
    )
   )
   &MeshRotationForce(Rotation=&RandomVector(Vector Scale=&RandomRangeFloat(Minimum=5, Maximum=10)))
   &ApplyInitialForces()
   &SphereLocation(
    MaskSpawnGroup=false
    Offset=<0 0 125>
    Override Seed=false
    Sphere Radius=0
    Surface Only=false
    UseNonUniformScale=false
    UseOffset=true
    UseSphereOrientation=false
   )
  ]
  particleUpdateScript [
   &ParticleState()
   &AccelerationForce(Acceleration=&MakeVector(
     X=0
     Y=0
     Z=&FN_Fit01(Clamp:NiagaraBool=false, Value:NiagaraFloat=$Particles.Distribution, DestMin:NiagaraFloat=-980, DestMax:NiagaraFloat=-1200)
    ))
   &Drag(Drag=0.1, Use Linear Drag=true, Use Rotational Drag=false)
   &GravityForce(Gravity=&Multiply_Vector(A=<1 1 -4000>, B=&VectorFromFloat(Value=$Particles.Distribution)))
   &SolveForcesAndVelocity(Clamp Velocity=false, Limit Acceleration=false)
   &SolveRotationalForcesAndVelocity()
   &FN_Mesh_Attributes(
    Create Scale Attribute=true
    Create Visibility Attribute=true
    Scale=&Multiply_Vector(
     A=$Particles.Initial.Scale
     B=&VectorFromFloat(Value=&FloatFromCurve(FloatCurve=DataInterface() {propertyValues {Curve {keys [
           <at -0.0032 0 RCIM_Cubic>
           {arriveTangent=-0.0978, interpMode=RCIM_Cubic, leaveTangent=-0.0978, tangentMode=RCTM_User, time=0.1249, value=0.9962}
           {arriveTangent=0.0267, interpMode=RCIM_Cubic, leaveTangent=0.0267, tangentMode=RCTM_User, time=0.7407, value=1.0021}
           <at 1 0 RCIM_Cubic>
          ]}}}))
    )
   )
  ]
  rendererClasses [&NiagaraMeshRendererProperties]
  renderers [{rendererClass=&NiagaraMeshRendererProperties, rendererIndex=0}]
 }
}
summary {
 systemName=NS_Glider_ChickenWarrior_Deploy
 emitters [
  {emitterName=Shell_Bits, simTarget=GPUComputeSim, rendererClasses [&NiagaraMeshRendererProperties]}
  {emitterName=Idle_Feathers, rendererClasses [&NiagaraMeshRendererProperties, &NiagaraSpriteRendererProperties]}
 ]
 userVariables [
  User.SkeletalMesh {
   type=NiagaraDataInterfaceSkeletalMesh
   defaultValue {
    struct=NiagaraExt_VariableValue_DataInterface
    value {dataInterface=&self:NiagaraDataInterfaceSkeletalMesh_0}
   }
  }
  User.VariantColor {type=LinearColor, description="", defaultValue {struct=LinearColor, value=<rgba 1 1 1 1>}}
  User.VariantNum {type=NiagaraInt32}
 ]
}
