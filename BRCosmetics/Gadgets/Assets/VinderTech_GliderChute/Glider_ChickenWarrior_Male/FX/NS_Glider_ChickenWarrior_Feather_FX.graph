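# Niagara effect that sheds drifting feathers from the Chicken Warrior glider, tinted by User.VariantColor.
# The only enabled emitter, Idle_Drops, spawns at a steady rate, using random spawn groups to pick skeletal mesh bones or sockets and inheriting capped velocity.
# Particles fall under curve-ramped gravity and drag, spin either way, and shrink at end of life and through a camera-based scratch module.
# A second idle emitter and a User.Land-driven burst emitter are disabled.
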NiagaraSystem /BRCosmetics/Gadgets/Assets/VinderTech_GliderChute/Glider_ChickenWarrior_Male/FX/NS_Glider_ChickenWarrior_Feather_FX
dependencies {
 usedDataInterfaces [
  &NiagaraDataInterfaceArrayFloat
  &NiagaraDataInterfaceCurve
  &NiagaraDataInterfaceCamera
  &NiagaraDataInterfaceArrayInt32
 ]
 usedDynamicInputs [&self:ScratchDynamicInput]
 usedModules [
  &SystemState
  &self:SystemSpawnScript.SystemScriptSource.SystemScriptGraph.NiagaraNodeAssignment_0.SetVariables_#1
 ]
}
emitters {
 Idle_Drops {
  emitterUpdateScript [
   &EmitterState()
   &SpawnRate(
    SpawnGroup:NiagaraInt32=&RandomRangeInt(Minimum=0, Maximum=4, Evaluation Type=::NewEnumerator1)
    SpawnRate=&Multiply_Float(
     A=&Lerp_Float(
      A=0
      B=8
      Alpha=&Power(a=&RandomRangeFloat(Minimum=0, Maximum=1, Evaluation Type=::NewEnumerator1), b=10)
     )
     B=$System.IdleSpawnRate
    )
    Use Spawn Probability=true
   ) {enabled=false}
   &SpawnRate(
    SpawnGroup:NiagaraInt32=&RandomRangeInt(Minimum=0, Maximum=8, Evaluation Type=::NewEnumerator1)
    SpawnRate=4
    Use Spawn Probability=false
   ) {module=SpawnRate001}
  ]
  particleSpawnScript [
   /Niagara/Modules/Spawn/Initialization/InitializeParticle(
    Color=$User.VariantColor
    Lifetime=&RandomRangeFloat(Minimum=1, Maximum=2)
    Mesh Scale=&Multiply_VectorByFloat(Vector=<1 1 1>, Float=1)
    Module.Write Color=true
    Module.Write Lifetime=true
    Module.Write Mass=true
    Module.Write Material Random=true
    Module.Write Position=false
    Module.Write Scale=true
    Module.Write SpriteRotation=false
    Module.Write SpriteSize=true
    Position=&SimulationPosition()
    Sprite Size=<1 1>
   )
   &SkeletalMeshLocation(
    Apply Sampled Orientation to Mesh Particles=true
    Bone / Socket Index=&SpawnGroup()
    Bone Sampling Mode=::NewEnumerator4
    Inherit Velocity=false
    MaskSpawnGroup=false
    Skeletal Mesh=$User.SkeletalMesh
   )
   &DynamicMaterialParameters(
    Index 0 Param 1=&SelectFloatFromArray(
     Direct Array Index=&SpawnGroup()
     Float Selection Array=DataInterface() {propertyValues {FloatData [0,0.75,0.25,0.75,0.75,0.5,0.75]}}
    )
    Index 0 Param 2=&FloatFromCurve(FloatCurve=DataInterface() {propertyValues {Curve {keys [<at 0 0>,<at 0.5 0.25>]}}})
    Index 0 Param 3=&FloatFromCurve(FloatCurve=DataInterface() {propertyValues {Curve {keys [<at 0 1>,<at 0.1 0.5>,<at 1 0>]}}})
    Index 0 Param 4=&FloatFromCurve(FloatCurve=DataInterface() {propertyValues {Curve {keys [<at 0 1>,<at 1 0>]}}})
    Module.Param0WriteEnabled=true
    Module.Param1WriteEnabled=false
    Module.Param2WriteEnabled=false
    Module.Param3WriteEnabled=false
   ) {module=DynamicMaterialParameters001}
   &AddVelocity(
    Coordinate Space=::Simulation
    Velocity=&RotateVectorByQuaternion(Vector To Rotate=<-50 0 0>, Quaternion=$Output.SkeletalMeshLocation.SampledRotation)
   ) {enabled=false}
   /SetVariables(
    Particles.NewQuat:Quat4f=$Output.SkeletalMeshLocation.SampledRotation
    Particles.NewVector:Vector3f=$Output.SkeletalMeshLocation.SampledPosition
    Particles.SpinDirection:NiagaraFloat=&UniformAOrBFloat(A=-1)
   ) {isSetParameters=true, module=SetVariables_#2}
   &InheritVelocity(Module.Limit Speed=true, Velocity Limit=300, Velocity Scale=&Multiply_VectorByFloat(Vector=<1 1 1>, Float=20))
  ]
  particleUpdateScript [
   &self:CameraSize(SomeValueDivisor=50, ZeroOffset=-75) {module=CameraSize}
   &ParticleState()
   &GravityForce(Gravity=&Multiply_VectorByFloat(
     Vector=<1 1 -250>
     Float=&FloatFromCurve(FloatCurve=DataInterface() {propertyValues {Curve {keys [{arriveTangent=2.0378, interpMode=RCIM_Cubic, leaveTangent=2.0378, tangentMode=RCTM_User, time=0, value=0}, <at 1 1>]}}})
    ))
   &Drag(
    Drag=&Multiply_Float(
     A=5
     B=&FloatFromCurve(FloatCurve=DataInterface() {propertyValues {Curve {keys [<at 0 0 RCIM_Cubic>, {arriveTangent=2.7826, interpMode=RCIM_Cubic, leaveTangent=2.7826, tangentMode=RCTM_User, time=1, value=1}]}}})
    )
    Use Linear Drag=true
    Use Rotational Drag=false
   )
   &AddVelocity(Velocity=<0 -1 0>) {module=AddVelocity002}
   &SolveForcesAndVelocity(Clamp Velocity=false, Limit Acceleration=false)
   &UpdateMeshOrientation(
    Rotation Coordinate Space=::NewEnumerator3
    Rotation Rate=&Multiply_Float(
     A=&Multiply_Float(A=&RandomRangeFloat(Minimum=1.5, Maximum=2), B=$Particles.SpinDirection)
     B=&FloatFromCurve(FloatCurve=DataInterface() {propertyValues {Curve {keys [
          {arriveTangent=0.0162, interpMode=RCIM_Cubic, leaveTangent=0.0162, tangentMode=RCTM_User, time=0, value=0}
          {arriveTangent=0.0117, interpMode=RCIM_Cubic, leaveTangent=0.0038, tangentMode=RCTM_Break, time=0.25, value=0.9425}
          <at 1 0.5>
         ]}}})
    )
    Rotation Vector=&Add_Vector(A=<1 1 1>, B=&RandomRangeVector(Minimum=<0 0 0>, Maximum=<0 0 0>))
   )
   &ScaleMeshSize(Scale Factor=&Multiply_Vector(
     A=&VectorFromFloat(Value=&FloatFromCurve(FloatCurve=DataInterface() {propertyValues {Curve {keys [<at 0.75 1 RCIM_Cubic>, {arriveTangent=-10.6879, interpMode=RCIM_Cubic, leaveTangent=-10.6879, tangentMode=RCTM_User, time=1, value=0}]}}}))
     B=&Lerp_Vector(A=<0 0 0>, B=<1 1 1>, Alpha=&ClampFloat(Float=$Particles.CameraShrink))
    ))
  ]
  rendererClasses [&NiagaraMeshRendererProperties, &NiagaraSpriteRendererProperties]
  renderers [
   {rendererClass=&NiagaraMeshRendererProperties, rendererIndex=0}
   {rendererClass=&NiagaraSpriteRendererProperties, rendererIndex=1}
  ]
 }
 Idle_Drops001 {
  bEnabled=false
  emitterUpdateScript [
   &EmitterState()
   &SpawnRate(
    Spawn Probability=0.75
    SpawnGroup:NiagaraInt32=&RandomRangeInt(Minimum=0, Maximum=1, Evaluation Type=::NewEnumerator1)
    SpawnRate=&Multiply_Float(A=1, B=$System.IdleSpawnRate)
    Use Spawn Probability=true
   )
  ]
  particleSpawnScript [
   &InitializeParticle(
    Color=<rgba 0.9301 0.7241 0.1938 1>
    Lifetime=&RandomRangeFloat(Minimum=1, Maximum=2)
    Mesh Scale=&Multiply_VectorByFloat(Vector=<1 1 1>, Float=0.75)
    Module.Write Color=true
    Module.Write Lifetime=true
    Module.Write Mass=true
    Module.Write Material Random=true
    Module.Write Position=false
    Module.Write Scale=true
    Module.Write SpriteRotation=false
    Module.Write SpriteSize=true
    Position=&SimulationPosition()
    Sprite Size=<1 1>
   )
   &SkeletalMeshLocation(
    Apply Sampled Orientation to Mesh Particles=true
    Bone / Socket Index=&SpawnGroup()
    Bone Sampling Mode=::NewEnumerator4
    Inherit Velocity=false
    MaskSpawnGroup=false
    Skeletal Mesh=$User.SkeletalMesh
   )
   &DynamicMaterialParameters(
    Index 0 Param 1=&SelectFloatFromArray(
     Direct Array Index=&SpawnGroup()
     Float Selection Array=DataInterface() {propertyValues {FloatData [0,0.75,0.25,0.75,0.75,0.5,0.75]}}
    )
    Index 0 Param 2=&FloatFromCurve(FloatCurve=DataInterface() {propertyValues {Curve {keys [<at 0 0>,<at 0.5 0.25>]}}})
    Index 0 Param 3=&FloatFromCurve(FloatCurve=DataInterface() {propertyValues {Curve {keys [<at 0 1>,<at 0.1 0.5>,<at 1 0>]}}})
    Index 0 Param 4=&FloatFromCurve(FloatCurve=DataInterface() {propertyValues {Curve {keys [<at 0 1>,<at 1 0>]}}})
    Module.Param0WriteEnabled=true
    Module.Param1WriteEnabled=false
    Module.Param2WriteEnabled=false
    Module.Param3WriteEnabled=false
   ) {module=DynamicMaterialParameters001}
   &AddVelocity(
    Coordinate Space=::Simulation
    Velocity=&RotateVectorByQuaternion(Vector To Rotate=<-50 0 0>, Quaternion=$Output.SkeletalMeshLocation.SampledRotation)
   )
   /SetVariables(
    Particles.NewQuat:Quat4f=$Output.SkeletalMeshLocation.SampledRotation
    Particles.NewVector:Vector3f=$Output.SkeletalMeshLocation.SampledPosition
    Particles.SpinDirection:NiagaraFloat=&UniformAOrBFloat(A=-1)
   ) {isSetParameters=true, module=SetVariables_#2}
   &InheritVelocity(Module.Limit Speed=true, Velocity Scale=&Multiply_VectorByFloat(Vector=<1 1 1>, Float=20))
   &AddVelocity(Velocity=&RandomRangeVector(Minimum=<0 -15 0>, Maximum=<1 -20 1>)) {module=AddVelocity001}
   /SetVariables(Particles.VisibilityTag=/Game/Effects/Niagara/DynamicInputs/Arrays/FN_SelectIntFromArray(
     Array Sampling Mode:ENiagaraArraySamplingMode=::NewEnumerator1
     Interpolation:NiagaraFloat=?
     Evaluation Type:ENiagaraRandomnessEvaluation=::NewEnumerator0
     Normalized Interpolation:NiagaraBool=true
     Direct Array Index:NiagaraInt32=&SpawnGroup()
     Recalculate Random Each Loop:NiagaraBool=false
     Int Selection Array:NiagaraDataInterfaceArrayInt32=DataInterface() {propertyValues {IntData [0,0,0,1,0,0,1]}}
     Randomness Mode:ENiagaraRandomnessMode=::NewEnumerator0
     Random Seed:NiagaraInt32=?
     Fixed Random Seed:NiagaraBool=false
    )) {isSetParameters=true, module=SetVariables_#3}
  ]
  particleUpdateScript [
   &self:CameraSize(SomeValueDivisor=50, ZeroOffset=-75) {module=CameraSize}
   &ParticleState()
   /SetVariables(
    Particles.MultiDirectWave:Vector3f=&MakeVector(
     X=&Sine(Normalized Angle=&Add_Float(A=$Emitter.Age, B=$Particles.Lifetime), Period=2, Scale=1, Bias=0)
     Y=&Sine(Normalized Angle=&Add_Float(A=$Emitter.Age, B=$Particles.Lifetime), Period=2, Scale=1, Bias=0)
     Z=1
    )
    Particles.SineInfluence:Vector3f=&Multiply_VectorByFloat(Vector=&RandomRangeVector(Minimum=<-1 -1 0>, Maximum=<1 1 0>), Float=0.1)
   ) {enabled=false, isSetParameters=true, module=SetVariables_#4}
   /SetVariables(Particles.DriftVelocityWave:Vector3f=&Add_Vector(
     A=&Multiply_VectorByFloat(Vector=&RandomRangeVector(Minimum=<-1 -1 0>, Maximum=<1 1 0>), Float=20)
     B=&Multiply_Vector(A=$Particles.SineInfluence, B=$Particles.MultiDirectWave)
    )) {enabled=false, isSetParameters=true, module=SetVariables_#5}
   /SetVariables(Particles.Velocity=&Add_Vector(A=$Particles.Velocity, B=$Particles.DriftVelocityWave)) {enabled=false, isSetParameters=true, module=SetVariables_#6}
   &GravityForce(Gravity=&Multiply_VectorByFloat(
     Vector=<1 1 -250>
     Float=&FloatFromCurve(FloatCurve=DataInterface() {propertyValues {Curve {keys [{arriveTangent=2.0378, interpMode=RCIM_Cubic, leaveTangent=2.0378, tangentMode=RCTM_User, time=0, value=0}, <at 1 1>]}}})
    ))
   &Drag(
    Drag=&Multiply_Float(
     A=5
     B=&FloatFromCurve(FloatCurve=DataInterface() {propertyValues {Curve {keys [<at 0 0 RCIM_Cubic>, {arriveTangent=2.7826, interpMode=RCIM_Cubic, leaveTangent=2.7826, tangentMode=RCTM_User, time=1, value=1}]}}})
    )
    Use Linear Drag=true
    Use Rotational Drag=false
   )
   &AddVelocity(Velocity=<0 -1 0>) {module=AddVelocity002}
   &SolveForcesAndVelocity(Clamp Velocity=false, Limit Acceleration=false)
   &DynamicMaterialParameters(
    Index 0 Param 1=0.75
    Index 0 Param 2=&FloatFromCurve(FloatCurve=DataInterface() {propertyValues {Curve {keys [<at 0 0>,<at 0.5 0.25>]}}})
    Index 0 Param 3=&FloatFromCurve(FloatCurve=DataInterface() {propertyValues {Curve {keys [<at 0 1>,<at 0.1 0.5>,<at 1 0>]}}})
    Index 0 Param 4=&FloatFromCurve(FloatCurve=DataInterface() {propertyValues {Curve {keys [
         <at 0 0.5 RCIM_Cubic>
         {arriveTangent=0.7684, interpMode=RCIM_Cubic, leaveTangent=0.7684, time=0.26, value=0.5807}
         {arriveTangent=-0.7847, interpMode=RCIM_Cubic, leaveTangent=-0.7847, time=0.5225, value=0.9015}
         <at 1 0>
        ]}}})
    Module.Param0WriteEnabled=false
    Module.Param1WriteEnabled=true
    Module.Param2WriteEnabled=true
    Module.Param3WriteEnabled=true
   )
   &DynamicMaterialParameters(
    Index 0 Param 1=0.75
    Index 0 Param 2=0.5
    Index 0 Param 3=0.5
    Index 0 Param 4=0.5
    Module.Param0WriteEnabled=false
    Module.Param1WriteEnabled=true
    Module.Param2WriteEnabled=true
    Module.Param3WriteEnabled=true
   ) {enabled=false, module=DynamicMaterialParameters002}
   &UpdateMeshOrientation(
    Rotation Coordinate Space=::NewEnumerator3
    Rotation Rate=&Multiply_Float(
     A=&Multiply_Float(A=&RandomRangeFloat(Minimum=1.5, Maximum=2), B=$Particles.SpinDirection)
     B=&FloatFromCurve(FloatCurve=DataInterface() {propertyValues {Curve {keys [
          {arriveTangent=0.0162, interpMode=RCIM_Cubic, leaveTangent=0.0162, tangentMode=RCTM_User, time=0, value=0}
          {arriveTangent=0.0117, interpMode=RCIM_Cubic, leaveTangent=0.0038, tangentMode=RCTM_Break, time=0.25, value=0.9425}
          <at 1 0.5>
         ]}}})
    )
    Rotation Vector=&Add_Vector(A=<1 1 1>, B=&RandomRangeVector(Minimum=<0 0 0>, Maximum=<0 0 0>))
   )
   &ScaleMeshSize(Scale Factor=&Multiply_Vector(
     A=&VectorFromFloat(Value=&FloatFromCurve(FloatCurve=DataInterface() {propertyValues {Curve {keys [<at 0.75 1 RCIM_Cubic>, {arriveTangent=-10.6879, interpMode=RCIM_Cubic, leaveTangent=-10.6879, tangentMode=RCTM_User, time=1, value=0}]}}}))
     B=&Lerp_Vector(A=<0 0 0>, B=<1 1 1>, Alpha=&ClampFloat(Float=$Particles.CameraShrink))
    ))
  ]
  rendererClasses [&NiagaraSpriteRendererProperties, &NiagaraMeshRendererProperties]
  renderers [
   {rendererClass=&NiagaraSpriteRendererProperties}
   {rendererClass=&NiagaraMeshRendererProperties, rendererIndex=1}
   {rendererClass=&NiagaraMeshRendererProperties, rendererIndex=2}
   {rendererClass=&NiagaraMeshRendererProperties, rendererIndex=3}
  ]
 }
 Jump_Drops001 {
  bEnabled=false
  emitterUpdateScript [
   &EmitterState(Life Cycle Mode=::NewEnumerator1, Loop Duration=1, Recalculate Duration Each Loop=true, UseLoopDelay=true)
   &SpawnRate(
    SpawnGroup:NiagaraInt32=&RandomRangeInt(Minimum=0, Maximum=6, Evaluation Type=::NewEnumerator1)
    SpawnRate=&Multiply_Float(A=150, B=$User.Land)
    Use Spawn Probability=true
   )
   &SpawnBurst_Instantaneous(
    Spawn Count=20
    Spawn Group:NiagaraInt32=&RandomRangeInt(Minimum=0, Maximum=6)
    Use Spawn Probability=false
   ) {enabled=false}
  ]
  particleSpawnScript [
   /SetVariables(Particles.SpawnGroup=&RandomRangeInt(Minimum=0, Maximum=1)) {isSetParameters=true}
   &InitializeParticle(
    Color=<rgba 0.9301 0.7241 0.1938 1>
    Lifetime=&RandomRangeFloat(Minimum=0.5, Maximum=2)
    Mesh Scale=&Multiply_VectorByFloat(Vector=<1 1 1>, Float=0.75)
    Module.Write Color=true
    Module.Write Lifetime=true
    Module.Write Mass=true
    Module.Write Material Random=true
    Module.Write Position=false
    Module.Write Scale=true
    Module.Write SpriteRotation=false
    Module.Write SpriteSize=true
    Position=&SimulationPosition()
    Sprite Size=<1 1>
   )
   &SkeletalMeshLocation(
    Apply Sampled Orientation to Mesh Particles=true
    Bone / Socket Index=&SpawnGroup()
    Bone Sampling Mode=::NewEnumerator4
    Inherit Velocity=false
    MaskSpawnGroup=false
    Skeletal Mesh=$User.SkeletalMesh
   )
   &DynamicMaterialParameters(
    Index 0 Param 1=&SelectFloatFromArray(
     Direct Array Index=&SpawnGroup()
     Float Selection Array=DataInterface() {propertyValues {FloatData [0,0.75,0.25,0.75,0.75,0.5,0.75]}}
    )
    Index 0 Param 2=&FloatFromCurve(FloatCurve=DataInterface() {propertyValues {Curve {keys [<at 0 0>,<at 0.5 0.25>]}}})
    Index 0 Param 3=&FloatFromCurve(FloatCurve=DataInterface() {propertyValues {Curve {keys [<at 0 1>,<at 0.1 0.5>,<at 1 0>]}}})
    Index 0 Param 4=&FloatFromCurve(FloatCurve=DataInterface() {propertyValues {Curve {keys [<at 0 1>,<at 1 0>]}}})
    Module.Param0WriteEnabled=true
    Module.Param1WriteEnabled=false
    Module.Param2WriteEnabled=false
    Module.Param3WriteEnabled=false
   ) {module=DynamicMaterialParameters001}
   &AddVelocity(
    Coordinate Space=::Simulation
    Velocity=&RotateVectorByQuaternion(
     Vector To Rotate=&RandomRangeVector(Minimum=<-25 0 0>, Maximum=<-100 1 1>)
     Quaternion=$Output.SkeletalMeshLocation.SampledRotation
    )
   )
   /SetVariables(
    Particles.NewQuat:Quat4f=$Output.SkeletalMeshLocation.SampledRotation
    Particles.NewVector:Vector3f=$Output.SkeletalMeshLocation.SampledPosition
    Particles.SpinDirection:NiagaraFloat=&UniformAOrBFloat(A=-1)
   ) {isSetParameters=true, module=SetVariables_#2}
   &AddVelocity(Velocity=&RandomRangeVector(Minimum=<0 -15 0>, Maximum=<1 -20 1>)) {module=AddVelocity001}
   &InheritVelocity(Module.Limit Speed=true, Velocity Scale=&Multiply_VectorByFloat(Vector=<1 1 0>, Float=20)) {enabled=false}
   /SetVariables(Particles.VisibilityTag=&FN_SelectIntFromArray(
     Array Sampling Mode:ENiagaraArraySamplingMode=::NewEnumerator1
     Interpolation:NiagaraFloat=?
     Evaluation Type:ENiagaraRandomnessEvaluation=::NewEnumerator0
     Normalized Interpolation:NiagaraBool=true
     Direct Array Index:NiagaraInt32=&SpawnGroup()
     Recalculate Random Each Loop:NiagaraBool=false
     Int Selection Array:NiagaraDataInterfaceArrayInt32=DataInterface() {propertyValues {IntData [0,0,0,1,0,0,1]}}
     Randomness Mode:ENiagaraRandomnessMode=::NewEnumerator0
     Random Seed:NiagaraInt32=?
     Fixed Random Seed:NiagaraBool=false
    )) {isSetParameters=true, module=SetVariables_#3}
  ]
  particleUpdateScript [
   &self:CameraSize_01(SomeValueDivisor=50, ZeroOffset=-75) {module=CameraSize_01}
   &ParticleState()
   /SetVariables(
    Particles.MultiDirectWave:Vector3f=&MakeVector(
     X=&Sine(Normalized Angle=&Add_Float(A=$Emitter.Age, B=$Particles.Lifetime), Period=2, Scale=1, Bias=0)
     Y=&Sine(Normalized Angle=&Add_Float(A=$Emitter.Age, B=$Particles.Lifetime), Period=2, Scale=1, Bias=0)
     Z=1
    )
    Particles.SineInfluence:Vector3f=&Multiply_VectorByFloat(Vector=&RandomRangeVector(Minimum=<-1 -1 0>, Maximum=<1 1 0>), Float=0.1)
   ) {enabled=false, isSetParameters=true, module=SetVariables_#4}
   /SetVariables(Particles.DriftVelocityWave:Vector3f=&Add_Vector(
     A=&Multiply_VectorByFloat(Vector=&RandomRangeVector(Minimum=<-1 -1 0>, Maximum=<1 1 0>), Float=20)
     B=&Multiply_Vector(A=$Particles.SineInfluence, B=$Particles.MultiDirectWave)
    )) {enabled=false, isSetParameters=true, module=SetVariables_#5}
   /SetVariables(Particles.Velocity=&Add_Vector(A=$Particles.Velocity, B=$Particles.DriftVelocityWave)) {enabled=false, isSetParameters=true, module=SetVariables_#6}
   &GravityForce(Gravity=&Multiply_VectorByFloat(
     Vector=<1 1 -250>
     Float=&FloatFromCurve(FloatCurve=DataInterface() {propertyValues {Curve {keys [{arriveTangent=2.0378, interpMode=RCIM_Cubic, leaveTangent=2.0378, tangentMode=RCTM_User, time=0, value=0}, <at 1 1>]}}})
    ))
   &Drag(
    Drag=&Multiply_Float(
     A=5
     B=&FloatFromCurve(FloatCurve=DataInterface() {propertyValues {Curve {keys [<at 0 0 RCIM_Cubic>, {arriveTangent=2.7826, interpMode=RCIM_Cubic, leaveTangent=2.7826, tangentMode=RCTM_User, time=1, value=1}]}}})
    )
    Use Linear Drag=true
    Use Rotational Drag=false
   )
   &AddVelocity(Velocity=<0 -1 0>) {module=AddVelocity002}
   &SolveForcesAndVelocity(Clamp Velocity=false, Limit Acceleration=false)
   &DynamicMaterialParameters(
    Index 0 Param 1=0.75
    Index 0 Param 2=&FloatFromCurve(FloatCurve=DataInterface() {propertyValues {Curve {keys [<at 0 0>,<at 0.5 0.25>]}}})
    Index 0 Param 3=&FloatFromCurve(FloatCurve=DataInterface() {propertyValues {Curve {keys [<at 0 1>,<at 0.1 0.5>,<at 1 0>]}}})
    Index 0 Param 4=&FloatFromCurve(FloatCurve=DataInterface() {propertyValues {Curve {keys [
         <at 0 0.5 RCIM_Cubic>
         {arriveTangent=0.7684, interpMode=RCIM_Cubic, leaveTangent=0.7684, time=0.26, value=0.5807}
         {arriveTangent=-0.7847, interpMode=RCIM_Cubic, leaveTangent=-0.7847, time=0.5225, value=0.9015}
         <at 1 0>
        ]}}})
    Module.Param0WriteEnabled=false
    Module.Param1WriteEnabled=true
    Module.Param2WriteEnabled=true
    Module.Param3WriteEnabled=true
   )
   &DynamicMaterialParameters(
    Index 0 Param 1=0.75
    Index 0 Param 2=0.5
    Index 0 Param 3=0.5
    Index 0 Param 4=0.5
    Module.Param0WriteEnabled=false
    Module.Param1WriteEnabled=true
    Module.Param2WriteEnabled=true
    Module.Param3WriteEnabled=true
   ) {enabled=false, module=DynamicMaterialParameters002}
   &UpdateMeshOrientation(
    Rotation Coordinate Space=::NewEnumerator3
    Rotation Rate=&Multiply_Float(
     A=&Multiply_Float(A=1.5, B=$Particles.SpinDirection)
     B=&FloatFromCurve(FloatCurve=DataInterface() {propertyValues {Curve {keys [
          {arriveTangent=6.9741, interpMode=RCIM_Cubic, leaveTangent=6.9741, tangentMode=RCTM_User, time=0, value=0}
          {arriveTangent=0.0117, interpMode=RCIM_Cubic, leaveTangent=0.0038, tangentMode=RCTM_Break, time=0.25, value=0.9425}
          <at 1 0.5>
         ]}}})
    )
    Rotation Vector=&Add_Vector(A=<1 1 1>, B=&RandomRangeVector(Minimum=<0.5 0 0>, Maximum=<1 0 1>))
   )
   &ScaleMeshSize(Scale Factor=&Multiply_Vector(
     A=&VectorFromFloat(Value=&FloatFromCurve(FloatCurve=DataInterface() {propertyValues {Curve {keys [<at 0.75 1 RCIM_Cubic>, {arriveTangent=-10.6879, interpMode=RCIM_Cubic, leaveTangent=-10.6879, tangentMode=RCTM_User, time=1, value=0}]}}}))
     B=&Lerp_Vector(A=<0 0 0>, B=<1 1 1>, Alpha=&ClampFloat(Float=$Particles.CameraShrink))
    ))
  ]
  rendererClasses [&NiagaraSpriteRendererProperties, &NiagaraMeshRendererProperties]
  renderers [
   {rendererClass=&NiagaraSpriteRendererProperties}
   {rendererClass=&NiagaraMeshRendererProperties, rendererIndex=1}
   {rendererClass=&NiagaraMeshRendererProperties, rendererIndex=2}
   {rendererClass=&NiagaraMeshRendererProperties, rendererIndex=3}
  ]
 }
}
summary {
 systemName=NS_Glider_ChickenWarrior_Feather_FX
 emitters [
  {emitterName=Idle_Drops, rendererClasses [&NiagaraMeshRendererProperties, &NiagaraSpriteRendererProperties]}
  {
   bEnabled=false
   emitterName=Jump_Drops001
   rendererClasses [&NiagaraSpriteRendererProperties, &NiagaraMeshRendererProperties]
  }
  {
   bEnabled=false
   emitterName=Idle_Drops001
   rendererClasses [&NiagaraSpriteRendererProperties, &NiagaraMeshRendererProperties]
  }
 ]
 userVariables [
  User.Land {type=NiagaraFloat, defaultValue=0, description=""}
  User.RunningSpeed {type=NiagaraFloat, defaultValue=0, description=""}
  User.SkeletalMesh {
   type=NiagaraDataInterfaceSkeletalMesh
   defaultValue {
    struct=NiagaraExt_VariableValue_DataInterface
    value {dataInterface=&self:NiagaraDataInterfaceSkeletalMesh_6}
   }
  }
  User.VariantColor {type=LinearColor, description="", defaultValue {struct=LinearColor, value=<rgba 1 1 1 1>}}
 ]
}
