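# Niagara one-shot Slurp splash effect for deploying the male Chicken Warrior glider, with every emitter spawning from locations sampled on the User skeletal mesh.
# Bubbles, drops and juice sprites burst outward from the sampled points, shaped by drag plus curl noise on bubbles and gravity on the liquid.
# Juice uses SubUV frames, animated material parameters and growing camera offset.
# A single expanding mesh flash fades from intense cyan to blue.
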NiagaraSystem /BRCosmetics/Gadgets/Assets/VinderTech_GliderChute/Glider_ChickenWarrior_Male/FX/NS_Glider_ChickenWarrior_Slurp_Deploy
dependencies {
 usedDataInterfaces [&NiagaraDataInterfaceCurve, &NiagaraDataInterfaceColorCurve]
 usedDynamicInputs []
 usedModules [&SystemState]
}
emitters {
 Bubbles {
  emitterUpdateScript [
   &EmitterState()
   &SpawnBurst_Instantaneous(Spawn Count=100, Spawn Group=1, Spawn Time=0.05, Use Spawn Probability=false)
  ]
  particleSpawnScript [
   &InitializeParticle(
    Color Mode=::NewEnumerator0
    Lifetime=&RandomRangeFloat(Minimum=0.2, Maximum=1)
    Mass Mode=::NewEnumerator1
    Sprite Size Mode=::NewEnumerator1
    Uniform Sprite Size Max=15
    Uniform Sprite Size Min=4
    UsePositionOffset=false
    Write Material Random=true
   )
   &SkeletalMeshLocation(
    Inherit Velocity=false
    MaskSpawnGroup=false
    Mesh Sampling Type=::NewEnumerator5
    Skeletal Mesh=$User.SkeletalMesh
    Socket Sampling Mode=::NewEnumerator5
   )
   &SphereLocation(
    MaskSpawnGroup=false
    Override Seed=false
    Sphere Radius=80
    Surface Only=false
    UseNonUniformScale=false
    UseOffset=false
    UseSphereOrientation=false
   )
   &AddVelocityFromPoint(
    Default Position:Vector3f=$Output.SkeletalMeshLocation.SampledPosition
    Module.Use Falloff Distance=false
    Origin Offset=&FN_FrontEndVector3Modifier(Gameplay=<0 0 0>, Front End=<0 0 0>)
    Velocity Strength=&RandomRangeFloat(Minimum=100, Maximum=2000)
   ) {enabled=false}
   &AddVelocityFromPoint(
    Module.Use Falloff Distance=false
    Origin Offset=&FN_FrontEndVector3Modifier(Gameplay=<0 0 0>, Front End=<0 0 0>)
    Velocity Origin:Vector3f=$Output.SkeletalMeshLocation.SampledPosition
    Velocity Strength=&RandomRangeFloat(Minimum=100, Maximum=2000)
   ) {module=AddVelocityFromPoint001}
   &InheritVelocity(Module.Limit Speed=false)
  ]
  particleUpdateScript [
   &ParticleState()
   &Color(
    Color=&RandomRangeLinearColor(Minimum=<rgba 0 4 1.2666 1>, Maximum=<rgba 0 1.79 4 1>)
    Scale Alpha=&FloatFromCurve(FloatCurve=DataInterface() {propertyValues {Curve {keys [
         {interpMode=RCIM_Cubic, tangentMode=RCTM_Break, time=0, value=1}
         {arriveTangent=-3.5002, interpMode=RCIM_Cubic, tangentMode=RCTM_Break, time=1, value=0}
        ]}}})
   )
   /Niagara/Modules/Update/Forces/CurlNoiseForce(Noise Strength=300, Pan Noise=false)
   &Drag(Drag=2, Use Linear Drag=true, Use Rotational Drag=false)
   &SolveForcesAndVelocity(Clamp Velocity=false, Limit Acceleration=false)
   &ScaleSpriteSize(Scale Factor=&Vector2DFromFloat(Value=&FloatFromCurve(FloatCurve=DataInterface() {propertyValues {Curve {keys [{arriveTangent=5.1866, interpMode=RCIM_Cubic, leaveTangent=5.1866, tangentMode=RCTM_User, time=0, value=0}, <at 1 2>]}}})))
  ]
  rendererClasses [&NiagaraSpriteRendererProperties]
  renderers [{rendererClass=&NiagaraSpriteRendererProperties}]
 }
 Slurp_Drops {
  emitterUpdateScript [
   &EmitterState(
    Enable Distance Culling=true
    Life Cycle Mode=::NewEnumerator1
    Loop Behavior=::NewEnumerator1
    Loop Duration=1
    Module.bUseMaxDistance=true
    Module.bUseMinDistance=false
    Scalability Mode=::NewEnumerator1
    UseLoopDelay=false
   )
   &SpawnBurst_Instantaneous(Spawn Count=&UniformRangedInt(Minimum=10, Maximum=20), Use Spawn Probability=false)
  ]
  particleSpawnScript [
   /Niagara/Modules/Spawn/Initialization/InitializeParticle(
    Color=&UniformRangedLinearColor(Minimum=<rgba 1 1 1 1>, Maximum=<rgba 1 1 0.2 1>)
    Lifetime=&UniformRangedFloat(Minimum=0.5, Maximum=0.75)
    Mass=&UniformRangedFloat(Minimum=0.75, Maximum=1.75)
    Module.Write Color=true
    Module.Write Lifetime=true
    Module.Write Mass=true
    Module.Write Material Random=true
    Module.Write Position=false
    Module.Write Scale=false
    Module.Write SpriteRotation=false
    Module.Write SpriteSize=true
    Position=&SimulationPosition()
    Sprite Size=&Vector2DFromFloat(Value=&UniformRangedFloat(Minimum=10, Maximum=20))
   )
   &SkeletalMeshLocation(
    Inherit Velocity=false
    MaskSpawnGroup=false
    Mesh Sampling Type=::NewEnumerator5
    Skeletal Mesh=$User.SkeletalMesh
    Socket Sampling Mode=::NewEnumerator5
   )
   &SphereLocation(
    MaskSpawnGroup=false
    Offset Coordinate Space:ENiagaraCoordinateSpace=::Simulation
    Override Seed=false
    Sphere Radius=50
    Surface Only=false
    UseNonUniformScale=false
    UseOffset=true
    UseSphereOrientation=false
   )
   &AddVelocityFromPoint(
    Module.Use Falloff Distance=false
    Origin Offset Coordinate Space=::Simulation
    Velocity Strength=&UniformRangedFloat(Minimum=600, Maximum=1500)
   ) {enabled=false}
   &AddVelocityFromPoint(
    Module.Use Falloff Distance=false
    Origin Offset=&FN_FrontEndVector3Modifier(Gameplay=<0 0 0>, Front End=<0 0 0>)
    Velocity Origin:Vector3f=$Output.SkeletalMeshLocation.SampledPosition
    Velocity Strength=&RandomRangeFloat(Minimum=100, Maximum=2000)
   ) {module=AddVelocityFromPoint001}
  ]
  particleUpdateScript [
   &ParticleState()
   &GravityForce(Gravity=<0 0 -2250>)
   &Drag(Drag=&UniformRangedFloat(Minimum=1, Maximum=3), Use Linear Drag=true, Use Rotational Drag=false)
   &ScaleColor(Scale Alpha=&Lerp_Float(A=1, B=0, Alpha=$Particles.NormalizedAge), ScaleA=true, ScaleRGB=true)
   &SolveForcesAndVelocity(Clamp Velocity=false, Limit Acceleration=false)
   &ScaleSpriteSizeBySpeed(Max Scale Factor=<1 5>, Min Scale Factor=<0 0.05>)
  ]
  rendererClasses [&NiagaraSpriteRendererProperties]
  renderers [{rendererClass=&NiagaraSpriteRendererProperties}]
 }
 Slurp_Juice {
  emitterUpdateScript [
   &EmitterState(Life Cycle Mode=::NewEnumerator1, Loop Behavior=::NewEnumerator1, Loop Duration=1, UseLoopDelay=false)
   &SpawnBurst_Instantaneous(Spawn Count=&RandomRangeInt(Minimum=20, Maximum=24), Use Spawn Probability=false)
  ]
  particleSpawnScript [
   /Niagara/Modules/Spawn/Initialization/V2/InitializeParticle(
    Color Mode=::NewEnumerator0
    Lifetime Max=0.9
    Lifetime Min=0.4
    Lifetime Mode=::NewEnumerator1
    Mass Mode=::NewEnumerator1
    Position Offset=<0 0 -20>
    Sprite Rotation Angle=&RandomRangeFloat(Minimum=0, Maximum=360)
    Sprite Rotation Mode=::NewEnumerator0
    Sprite Size Mode=::NewEnumerator1
    Uniform Sprite Size Max=180
    Uniform Sprite Size Min=90
    UsePositionOffset=true
    Write Material Random=true
   )
   &SkeletalMeshLocation(
    Inherit Velocity=false
    MaskSpawnGroup=false
    Mesh Sampling Type=::NewEnumerator5
    Skeletal Mesh=$User.SkeletalMesh
    Socket Sampling Mode=::NewEnumerator5
   )
   &SphereLocation(
    MaskSpawnGroup=false
    Override Seed=false
    Sphere Radius=75
    Surface Only=false
    UseNonUniformScale=false
    UseOffset=false
    UseSphereOrientation=false
   )
   &AddVelocityFromPoint(
    Module.Use Falloff Distance=false
    Origin Offset=&FN_FrontEndVector3Modifier(Gameplay=<0 0 0>, Front End=<0 0 0>)
    Velocity Origin:Vector3f=$Output.SkeletalMeshLocation.SampledPosition
    Velocity Strength=&RandomRangeFloat(Minimum=350, Maximum=1500)
   )
   &ApplyInitialForces()
   &CameraOffset(Camera Offset Amount=&RandomRangeFloat(Minimum=30, Maximum=45)) {module=CameraOffset001}
   &SubUVAnimation(End Frame=3, SubUV Animation Mode:ENiagara_SubUVLookupMode=::NewEnumerator2)
  ]
  particleUpdateScript [
   &ParticleState()
   &DynamicMaterialParameters(
    Index 0 Param 1=&FloatFromCurve(
     FloatCurve=DataInterface() {propertyValues {Curve {keys [
         {arriveTangent=-2.6317, interpMode=RCIM_Cubic, leaveTangent=-2.6317, tangentMode=RCTM_User, time=0, value=1}
         {interpMode=RCIM_Cubic, leaveTangent=-3.3696, time=1, value=0}
        ]}}}
     Scale Curve=3
    )
    Index 0 Param 2=&FloatFromCurve(
     FloatCurve=DataInterface() {propertyValues {Curve {keys [{arriveTangent=0.4361, interpMode=RCIM_Cubic, leaveTangent=0.4361, tangentMode=RCTM_User, time=0, value=0.05}, <at 1 0.25>]}}}
     Scale Curve=0.25
    )
    Module.Param0WriteEnabled=true
    Module.Param1WriteEnabled=true
    Module.Param2WriteEnabled=true
    Module.Param3WriteEnabled=true
   )
   &GravityForce(Gravity=<0 0 -1200>)
   &Drag(Use Linear Drag=true, Use Rotational Drag=false)
   &ScaleSpriteSize(Scale Factor=&Vector2DFromFloat(Value=&FloatFromCurve(
      FloatCurve=DataInterface() {propertyValues {Curve {keys [
          {arriveTangent=16027.7041, interpMode=RCIM_Cubic, leaveTangent=0.9223, tangentMode=RCTM_User, time=0, value=0}
          {interpMode=RCIM_Cubic, leaveTangent=0.1291, time=1, value=1}
         ]}}}
      Scale Curve=1.5
     )))
   &SolveForcesAndVelocity(Clamp Velocity=false, Limit Acceleration=false)
   &SpriteRotationRate(Rotation Rate=&Lerp_Float(A=-150, B=150, Alpha=$Particles.Initial.MaterialRandom))
   &CameraOffset(Camera Offset Amount=&Lerp_Float(
     A=$Particles.Initial.CameraOffset
     B=&Add_Float(A=$Particles.Initial.CameraOffset, B=100)
     Alpha=$Particles.NormalizedAge
    ))
  ]
  rendererClasses [&NiagaraSpriteRendererProperties]
  renderers [{rendererClass=&NiagaraSpriteRendererProperties}]
 }
 initSphere {
  emitterUpdateScript [
   &EmitterState(
    Enable Distance Culling=true
    MaxDistance=13000
    Module.bUseMaxDistance=true
    Module.bUseMinDistance=false
    Scalability Mode=::NewEnumerator1
   )
   &SpawnBurst_Instantaneous(Spawn Count=1, Spawn Time=0.05, Use Spawn Probability=false)
  ]
  particleSpawnScript [
   /Niagara/Modules/Spawn/Initialization/InitializeParticle(
    Color=<rgba 0.5 0.5 0.5 0.5>
    Lifetime=0.5
    Mesh Scale=&VectorFromFloat(Value=1)
    Module.Write Color=true
    Module.Write Lifetime=true
    Module.Write Mass=true
    Module.Write Material Random=true
    Module.Write Position=true
    Module.Write Scale=true
    Module.Write SpriteRotation=false
    Module.Write SpriteSize=false
    Position=&Add_Vector(A=/Niagara/DynamicInputs/Helpers/SimulationPosition(), B=<0 0 10>)
    Sprite Size=<5 5>
   )
   &SkeletalMeshLocation(
    Inherit Velocity=false
    MaskSpawnGroup=false
    Mesh Sampling Type=::NewEnumerator5
    Skeletal Mesh=$User.SkeletalMesh
    Socket Sampling Mode=::NewEnumerator5
   )
  ]
  particleUpdateScript [
   &ParticleState()
   &ScaleMeshSize(Scale Factor=&VectorFromFloat(Value=&FloatFromCurve(
      FloatCurve=DataInterface() {propertyValues {Curve {keys [
          {interpMode=RCIM_Cubic, leaveTangent=2.1274, tangentMode=RCTM_Break, time=0, value=0}
          {arriveTangent=0.6694, interpMode=RCIM_Cubic, leaveTangent=0.6694, tangentMode=RCTM_User, time=0.494, value=0.654}
          <at 1.5 1>
         ]}}}
      Scale Curve=15
     )))
   &ScaleColor(
    Scale Alpha=&FloatFromCurve(FloatCurve=DataInterface() {propertyValues {Curve {keys [
         {interpMode=RCIM_Cubic, tangentMode=RCTM_Break, time=0, value=1}
         {arriveTangent=-2.5583, interpMode=RCIM_Cubic, tangentMode=RCTM_Break, time=1, value=0}
        ]}}})
    Scale RGB=&MakeVectorFromLinearColorRGB(Color=&ColorFromCurve(ColorCurve=DataInterface() {propertyValues {
        AlphaCurve {keys [{arriveTangent=0.9905, interpMode=RCIM_Cubic, time=0, value=1}, <at 1 1 RCIM_Cubic>]}
        BlueCurve {keys [<at 0 25>,<at 0.2107 1.25 RCIM_Cubic>]}
        GreenCurve {keys [<at 0 24.9999>, {interpMode=RCIM_Cubic, leaveTangent=-1.3137, time=0.2107, value=0.561}]}
        RedCurve {keys [<at 0 0>,<at 0.2107 0 RCIM_Cubic>]}
       }}))
    ScaleA=true
    ScaleRGB=true
   )
   /SetVariables(Particles.LightColor=&ColorFromCurve(ColorCurve=DataInterface() {propertyValues {
       AlphaCurve {keys [<at 0.5925 1 RCIM_Cubic>,<at 1 0 RCIM_Cubic>]}
       BlueCurve {keys [<at -0.0022 0>, {arriveTangent=1.6758, interpMode=RCIM_Cubic, leaveTangent=1.6758, time=0.0394, value=1}, <at 0.5946 1>, <at 1 0 RCIM_Cubic>]}
       GreenCurve {keys [
         <at -0.0022 0>
         {arriveTangent=0.2015, interpMode=RCIM_Cubic, leaveTangent=0.2015, time=0.0394, value=0.7569}
         <at 0.5946 0.1203>
         <at 1 0 RCIM_Cubic>
        ]}
       RedCurve {keys [<at -0.0022 0>,<at 0.0394 0.2292 RCIM_Cubic>,<at 0.5946 0>,<at 1 0 RCIM_Cubic>]}
      }})) {isSetParameters=true}
  ]
  rendererClasses [&NiagaraMeshRendererProperties]
  renderers [{rendererClass=&NiagaraMeshRendererProperties, rendererIndex=0}]
 }
}
summary {
 systemName=NS_Glider_ChickenWarrior_Slurp_Deploy
 emitters [
  {emitterName=Bubbles, rendererClasses [&NiagaraSpriteRendererProperties]}
  {emitterName=Slurp_Juice, rendererClasses [&NiagaraSpriteRendererProperties]}
  {emitterName=Slurp_Drops, rendererClasses [&NiagaraSpriteRendererProperties]}
  {emitterName=initSphere, rendererClasses [&NiagaraMeshRendererProperties]}
 ]
 userVariables [
  User.bIsFrontEnd {type=NiagaraBool}
  User.bIsFrontEndPreview {type=NiagaraBool}
  User.SkeletalMesh {
   type=NiagaraDataInterfaceSkeletalMesh
   defaultValue {
    struct=NiagaraExt_VariableValue_DataInterface
    value {dataInterface=&self:NiagaraDataInterfaceSkeletalMesh_0}
   }
  }
  User.VariantColor {type=LinearColor, description="", defaultValue {struct=LinearColor, value=<rgba 1 1 1 1>}}
 ]
}
