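# Niagara system that drips slurp liquid from the Chicken Warrior male glider.
# Drip_Wing spawns sprites on the skeletal mesh at a rate driven by running speed and cut while skydiving; SyrupDrops adds sprite and mesh drops using feather bones, tinted by a depth-difference curve.
# In gameplay both blend from glider-local to simulated positions as particles age, with front-end overrides.
# The leg and secondary drip emitters are disabled.
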NiagaraSystem /BRCosmetics/Gadgets/Assets/VinderTech_GliderChute/Glider_ChickenWarrior_Male/FX/NS_Glider_ChickenWarrior_Slurp_FX
dependencies {
 usedDataInterfaces [
  &NiagaraDataInterfaceVectorCurve
  &NiagaraDataInterfaceVector2DCurve
  &NiagaraDataInterfaceCurve
  &NiagaraDataInterfaceArrayFloat
  &NiagaraDataInterfaceCamera
  &NiagaraDataInterfaceColorCurve
 ]
 usedDynamicInputs []
 usedModules [&SystemState]
}
emitters {
 Drip_A_Legs {
  bEnabled=false
  emitterUpdateScript [
   &EmitterState(Life Cycle Mode=::NewEnumerator1, Loop Duration=2, UseLoopDelay=false)
   &SpawnBurst_Instantaneous(Spawn Count=1, Spawn Probability=0.3, Use Spawn Probability=true)
   &SpawnRate(Spawn Probability=0.3, SpawnRate=500, Use Spawn Probability=true)
  ]
  particleSpawnScript [
   /Niagara/Modules/Spawn/Initialization/InitializeParticle(
    Lifetime=&UniformRangedFloat(Minimum=1.8, Maximum=2.1)
    Mass=&UniformRangedFloat(Minimum=0.75, Maximum=1.75)
    Module.Write Color=true
    Module.Write Lifetime=true
    Module.Write Mass=false
    Module.Write Material Random=true
    Module.Write Position=false
    Module.Write Scale=false
    Module.Write SpriteRotation=true
    Module.Write SpriteSize=true
    Position=&SimulationPosition()
    Sprite Rotation=180
    Sprite Size=&Vector2DFromFloat(Value=&MakeFloatFromVector(Vector=&UniformRangedVector(Minimum=<7 0 0>, Maximum=<11 1 1>), Channel=::NewEnumerator0))
   )
   &SkeletalMeshLocation(
    Apply To Particle Position=false
    Bone / Socket Index=&RandomRangeInt(Minimum=40, Maximum=69)
    Bone Sampling Mode=::NewEnumerator6
    Inherit Velocity=false
    MaskSpawnGroup=false
    Skeletal Mesh=$User.SkeletalMesh
   )
   /SetVariables(Particles.StoredBoneIndex:NiagaraInt32=$Output.SkeletalMeshLocation.SampledBoneIndex) {isSetParameters=true}
  ]
  particleUpdateScript [
   &SkeletalMeshLocation(
    Bone / Socket Index=$Particles.Initial.StoredBoneIndex
    Bone Sampling Mode=::NewEnumerator5
    Inherit Velocity=false
    MaskSpawnGroup=false
    Skeletal Mesh=$User.SkeletalMesh
   ) {module=SkeletalMeshLocation001}
   &CameraOffset()
   &ParticleState()
   &AccelerationForce(Acceleration=&VectorFromCurve(VectorCurve=DataInterface() {propertyValues {
       XCurve {keys [<at 0 0>,<at 0.9989 0>]}
       YCurve {keys [<at 0 0>,<at 0.9989 0>]}
       ZCurve {keys [
         {arriveTangent=-15.5802, interpMode=RCIM_Cubic, leaveTangent=-8.2205, tangentMode=RCTM_User, time=0, value=-1}
         {arriveTangent=-134.1077, interpMode=RCIM_Cubic, leaveTangent=-221.0645, tangentMode=RCTM_User, time=0.3, value=-20}
        ]}
      }}))
   &SolveForcesAndVelocity(Clamp Velocity=false, Limit Acceleration=false)
   &ScaleSpriteSize(Scale Factor=&Vector2DFromCurve(
     Vector2Curve=DataInterface() {propertyValues {
       XCurve {keys [
         {interpMode=RCIM_Cubic, tangentMode=RCTM_User, time=0, value=1.15}
         {arriveTangent=-0.8621, interpMode=RCIM_Cubic, leaveTangent=-0.8621, tangentMode=RCTM_User, time=0.6, value=1.14}
         <at 0.8265 0.457>
         {arriveTangent=-0.351, interpMode=RCIM_Cubic, leaveTangent=-0.351, tangentMode=RCTM_User, time=0.87, value=0.4}
         {interpMode=RCIM_Cubic, tangentMode=RCTM_User, time=0.9989, value=1}
        ]}
       YCurve {keys [
         {interpMode=RCIM_Cubic, tangentMode=RCTM_User, time=0, value=0}
         {arriveTangent=2.8736, interpMode=RCIM_Cubic, leaveTangent=2.8736, tangentMode=RCTM_User, time=0.6, value=0.99}
         {arriveTangent=2.025, interpMode=RCIM_Cubic, leaveTangent=2.025, tangentMode=RCTM_User, time=0.87, value=2.5}
         {interpMode=RCIM_Cubic, tangentMode=RCTM_User, time=1, value=1.8}
        ]}
      }}
     ScaleCurve=<1 1>
    ))
   &ScaleColor(
    Scale Alpha=&FloatFromCurve(FloatCurve=DataInterface() {propertyValues {Curve {keys [<at 0 0>,<at 0.1 1>,<at 0.9 1>,<at 1.0011 0.002>]}}})
    ScaleA=true
    ScaleRGB=true
   )
   &GenerateDeathEvent(
    Gap Correction Amount=1.5
    Position To Send:Vector3f=/Niagara/DynamicInputs/Transforms/TransformPosition(Position=Linked() {linkedVariable {name=Particles.Position, type=NiagaraPosition}}, Source Space=::Local, Destination Space=::World)
   )
  ]
  rendererClasses [&NiagaraSpriteRendererProperties]
  renderers [{rendererClass=&NiagaraSpriteRendererProperties}]
 }
 Drip_B {
  bEnabled=false
  emitterUpdateScript [
   &EmitterState(Life Cycle Mode=::NewEnumerator1, Loop Duration=1, UseLoopDelay=false)
   &SpawnRate(SpawnRate=500, Use Spawn Probability=false)
  ]
  particleSpawnScript [
   &InitializeParticle(
    Color=<rgba 1 1 1 0.5>
    Lifetime=&UniformRangedFloat(Minimum=0.6, Maximum=0.8)
    Mass=&UniformRangedFloat(Minimum=0.75, Maximum=1.75)
    Module.Write Color=true
    Module.Write Lifetime=true
    Module.Write Mass=false
    Module.Write Material Random=true
    Module.Write Position=false
    Module.Write Scale=false
    Module.Write SpriteRotation=true
    Module.Write SpriteSize=true
    Position=&SimulationPosition()
    Sprite Rotation=180
    Sprite Size=&Vector2DFromFloat(Value=&MakeFloatFromVector(Vector=&UniformRangedVector(Minimum=<11 0 0>, Maximum=<14 1 1>), Channel=::NewEnumerator0))
   )
   /Niagara/Modules/Spawn/Location/SphereLocation(MaskSpawnGroup=false, Module.Override Randomness=false, Sphere Radius=0, Surface Only=false)
   &InheritVelocity(Module.Limit Speed=false, Velocity Scale=<0 0 1>)
  ]
  particleUpdateScript [
   &CameraOffset(Camera Offset Amount=6)
   &ParticleState()
   &GravityForce(Gravity=<0 0 -1200>)
   &SolveForcesAndVelocity(Clamp Velocity=false, Limit Acceleration=false)
   &ScaleSpriteSize(Scale Factor=&Vector2DFromCurve(
     Vector2Curve=DataInterface() {propertyValues {
       XCurve {keys [
         {arriveTangent=5.9728, interpMode=RCIM_Cubic, leaveTangent=5.9728, tangentMode=RCTM_User, time=0, value=0.1}
         {arriveTangent=-0.1, interpMode=RCIM_Cubic, leaveTangent=-0.1, tangentMode=RCTM_User, time=0.18, value=0.99}
         {interpMode=RCIM_Cubic, tangentMode=RCTM_User, time=1, value=0}
        ]}
       YCurve {keys [
         {arriveTangent=6.2715, interpMode=RCIM_Cubic, leaveTangent=6.2715, tangentMode=RCTM_User, time=0, value=0.1}
         {arriveTangent=-0.1001, interpMode=RCIM_Cubic, leaveTangent=-0.1001, tangentMode=RCTM_User, time=0.18, value=0.99}
         {interpMode=RCIM_Cubic, tangentMode=RCTM_User, time=0.9989, value=0}
        ]}
      }}
     ScaleCurve=<1 1>
    ))
  ]
  rendererClasses [&NiagaraSpriteRendererProperties]
  renderers [{rendererClass=&NiagaraSpriteRendererProperties}]
 }
 Drip_Wing {
  emitterUpdateScript [
   &EmitterState(Life Cycle Mode=::NewEnumerator1, Loop Duration=1, UseLoopDelay=false)
   &SpawnRate(
    Spawn Probability=0.25
    SpawnRate=&Multiply_Float(
     A=&Lerp_Float(
      A=3
      B=6
      Alpha=&ClampFloat(Float=&Multiply_Float(A=$User.runningSpeed, B=0.0018))
     )
     B=&MakeCustomFloatFromBool(Bool=$User.IsSkydiving, True Float=0, False Float=1)
    )
    Use Spawn Probability=true
   )
   /SetVariables(Emitter.PlayerVelocityVector:Vector3f=$Engine.Owner.Velocity) {isSetParameters=true}
   /SetVariables(Emitter.PlayerSpeed:NiagaraFloat=&VectorLength(Vector=$Emitter.PlayerVelocityVector)) {isSetParameters=true}
  ]
  particleSpawnScript [
   &InitializeParticle(
    Color=<rgba 1 1 1 0.5>
    Lifetime=&UniformRangedFloat(Minimum=1.25, Maximum=2.2)
    Mass=&UniformRangedFloat(Minimum=0.75, Maximum=1.75)
    Module.Write Color=true
    Module.Write Lifetime=true
    Module.Write Mass=false
    Module.Write Material Random=true
    Module.Write Position=false
    Module.Write Scale=false
    Module.Write SpriteRotation=true
    Module.Write SpriteSize=true
    Position=&SimulationPosition()
    Sprite Rotation=180
    Sprite Size=&Vector2DFromFloat(Value=&MakeFloatFromVector(Vector=&UniformRangedVector(Minimum=<6 0 0>, Maximum=<14 1 1>), Channel=::NewEnumerator0))
   )
   &InheritVelocity(Module.Limit Speed=true, Velocity Limit=0.1, Velocity Scale=<0 0 -0.01>) {enabled=false}
   &SkeletalMeshLocation(
    Bone / Socket Index=&RandomRangeInt(Minimum=0, Maximum=8)
    Bone Sampling Mode=::NewEnumerator4
    Inherit Velocity=false
    MaskSpawnGroup=false
    Sampled Position Offset=&Add_Vector(
     A=<-2.5 -11 0>
     B=&Multiply_Vector(A=&RandomVector(Vector Scale=11), B=<1 1 1.2>)
    )
    Skeletal Mesh=$User.SkeletalMesh
   )
   /SetVariables(Particles.LocalPosition:Vector3f=&TransformPosition(Position=Linked() {linkedVariable {name=Particles.Position, type=NiagaraPosition}}, Source Space=::World, Destination Space=::Local)) {isSetParameters=true}
  ]
  particleUpdateScript [
   &CameraOffset()
   &ParticleState()
   &GravityForce(Gravity=<0 0 9.8>) {enabled=false}
   &SolveForcesAndVelocity(Clamp Velocity=false, Limit Acceleration=false)
   &ScaleSpriteSize(Scale Factor=&Vector2DFromCurve(
     Vector2Curve=DataInterface() {propertyValues {
       XCurve {keys [
         {arriveTangent=5.9728, interpMode=RCIM_Cubic, leaveTangent=5.9728, tangentMode=RCTM_User, time=0, value=0.1}
         {arriveTangent=-0.1, interpMode=RCIM_Cubic, leaveTangent=-0.1, tangentMode=RCTM_User, time=0.25, value=0.99}
         {interpMode=RCIM_Cubic, tangentMode=RCTM_User, time=1, value=0}
        ]}
       YCurve {keys [
         {arriveTangent=6.2715, interpMode=RCIM_Cubic, leaveTangent=6.2715, tangentMode=RCTM_User, time=0, value=0.1}
         {arriveTangent=-0.1001, interpMode=RCIM_Cubic, leaveTangent=-0.1001, tangentMode=RCTM_User, time=0.25, value=0.99}
         {interpMode=RCIM_Cubic, tangentMode=RCTM_User, time=0.9989, value=0}
        ]}
      }}
     ScaleCurve=<1 1>
    ))
   /SetVariables(Particles.Position=&FN_FrontEndVector3Modifier(
     Gameplay=&Lerp_Vector(
      A=&TransformPosition(Position=$Particles.LocalPosition, Source Space=::Local, Destination Space=::Simulation)
      B=Linked() {linkedVariable {name=Particles.Position, type=NiagaraPosition}}
      Alpha=&FloatFromCurve(FloatCurve=DataInterface() {propertyValues {Curve {keys [
           {interpMode=RCIM_Cubic, leaveTangent=4.3479, tangentMode=RCTM_Break, time=0, value=0}
           {arriveTangent=0.8633, interpMode=RCIM_Cubic, tangentMode=RCTM_Break, time=0.5, value=1}
          ]}}})
     )
     Front End=Linked() {linkedVariable {name=Particles.Position, type=NiagaraPosition}}
    )) {isSetParameters=true}
  ]
  rendererClasses [&NiagaraSpriteRendererProperties]
  renderers [
   {rendererClass=&NiagaraSpriteRendererProperties}
   {rendererClass=&NiagaraSpriteRendererProperties, rendererIndex=1}
  ]
 }
 SyrupDrops {
  emitterUpdateScript [
   &EmitterState(
    Life Cycle Mode=::NewEnumerator1
    Loop Duration=&RandomRangeFloat(Minimum=0.25, Maximum=2)
    Recalculate Duration Each Loop=true
    UseLoopDelay=false
   )
   &SpawnRate(SpawnRate=2, Use Spawn Probability=false)
   /SetVariables(Emitter.PlayerVelocityVector:Vector3f=$Engine.Owner.Velocity) {isSetParameters=true}
   /SetVariables(Emitter.PlayerSpeed:NiagaraFloat=&VectorLength(Vector=$Emitter.PlayerVelocityVector)) {isSetParameters=true}
   &SpawnPerUnit(
    Module.Use Max Movement Threshold=true
    Module.Use Movement Tolerance=false
    Spawn Per Unit:NiagaraFloat=&Lerp_Float(
     A=40
     B=100
     Alpha=&FloatFromCurve(FloatCurve=DataInterface() {propertyValues {Curve {keys [<at 0 1>,<at 1 0>]}}}, CurveIndex=$Emitter.PlayerSpeed)
    )
    Use Spawn Probability=false
   ) {enabled=false}
   &SpawnBurst_Instantaneous(Spawn Count=1, Spawn Probability=0.5, Use Spawn Probability=true)
  ]
  particleSpawnScript [
   /SetVariables(Particles.WhichSocket:NiagaraInt32=&RandomRangeInt(Minimum=0, Maximum=6)) {isSetParameters=true}
   &InitializeParticle(
    Color=<rgba 1 1 1 2>
    Lifetime=&Multiply_Float(
     A=&FN_FrontEndFloatModifier(Gameplay=1, Front End=0.2)
     B=&RandomRangeFloat(Minimum=2, Maximum=3)
    )
    Mesh Scale=&Multiply_VectorByFloat(
     Vector=&Multiply_VectorByFloat(
      Vector=<1 1 1>
      Float=&Multiply_Float(
       A=1
       B=&SelectFloatFromArray(Direct Array Index=$Particles.WhichSocket, Float Selection Array=DataInterface() {propertyValues {FloatData [0.9,1,0.75,0.55,0.5,0.7,0.5]}})
      )
     )
     Float=&RandomRangeFloat(Minimum=1, Maximum=1.3)
    )
    Module.Write Color=true
    Module.Write Lifetime=true
    Module.Write Mass=true
    Module.Write Material Random=true
    Module.Write Position=false
    Module.Write Scale=true
    Module.Write SpriteRotation=false
    Module.Write SpriteSize=true
    Position=&SimulationPosition()
    Sprite Size=<5 5>
   )
   &SkeletalMeshLocation(
    Bone Sampling Mode=::NewEnumerator0
    Inherit Velocity=false
    Interpolate Mesh Sampling=false
    MaskSpawnGroup=false
    Module.Override Randomness=false
    Sampled Position Offset=<0 0 -1>
    Skeletal Mesh=DataInterface() {propertyValues {
      DefaultMesh=null
      ExcludeBoneName=root
      OptionalComponentClass=null
      PreviewMesh=/BRCosmetics/Gadgets/Assets/VinderTech_GliderChute/Glider_FlapjackWranglerMale/Meshes/Flapjack_Wrangler_Male_Glider
      SkinningMode=SkinOnTheFly
      SoftSourceActor=null
      SourceMode=Default
      UvSetIndex=0
      WholeMeshLOD=-1
      bExcludeBone=false
      bReadDeformedGeometry=true
      bRequireCurrentFrameData=true
      ComponentTags []
      FilteredBones [feather_outer1_01_l, feather_outer1_01_r, feather_mid_01_l, feather_mid_01_r, feather_inner_01_l, feather_inner_01_r]
      FilteredSockets []
      MeshUserParameter {parameter {name=None, typeDefHandle {registeredTypeIndex=1}, varData []}}
      SamplingRegions []
     }}
   )
   /SetVariables(Particles.LocalPosition:Vector3f=&TransformPosition(Position=Linked() {linkedVariable {name=Particles.Position, type=NiagaraPosition}}, Source Space=::World, Destination Space=::Local)) {isSetParameters=true}
   &ApplyInitialForces()
   &AddVelocity(Velocity=&FN_FrontEndVector3Modifier(Gameplay=<0 500 0>, Front End=<0 0 0>))
   /SetVariables(Particles.SocketFloat:NiagaraFloat=&MakeFloatFromInt(Int=$Particles.WhichSocket)) {isSetParameters=true}
  ]
  particleUpdateScript [
   /Niagara/Modules/Update/Orientation/V2/OrientMeshToVector(Override Quat=false, UseOptionalSideAxis=true, UseOptionalUpAxis=true)
   &InheritVelocity(Module.Limit Speed=true, Velocity Scale=&Multiply_VectorByFloat(Vector=<1 1 1>, Float=-0.1)) {module=InheritVelocity001}
   &ParticleState()
   &GravityForce(Gravity=&Multiply_VectorByFloat(
     Vector=<1 1 -700>
     Float=&FloatFromCurve(FloatCurve=DataInterface() {propertyValues {Curve {keys [{arriveTangent=1.8039, interpMode=RCIM_Cubic, leaveTangent=1.8039, tangentMode=RCTM_User, time=0, value=0}, <at 1 1>]}}})
    )) {enabled=false}
   &InheritVelocity(
    Module.Limit Speed=true
    Velocity Limit=150
    Velocity Scale=&Multiply_VectorByFloat(
     Vector=<1 1 1>
     Float=/Niagara/DynamicInputs/Math/LerpMultipleFloats(
      Float Count:ENiagara_MultipleLerpCount=::NewEnumerator3
      1:NiagaraFloat=0.02
      2:NiagaraFloat=0
      3:NiagaraFloat=?
      4:NiagaraFloat=?
      5:NiagaraFloat=?
      Alpha:NiagaraFloat=$Particles.NormalizedAge
     )
    )
   ) {enabled=false}
   &ScaleMeshSize(Scale Factor=&Multiply_Vector(
     A=<1 1 1>
     B=&VectorFromCurve(VectorCurve=DataInterface() {propertyValues {
        XCurve {keys [
          <at 0 1 RCIM_Cubic>
          {arriveTangent=0.5, interpMode=RCIM_Cubic, leaveTangent=-5.3087, tangentMode=RCTM_Break, time=0.2366, value=1.9311}
          {arriveTangent=11.709, interpMode=RCIM_Cubic, leaveTangent=11.709, tangentMode=RCTM_User, time=0.3386, value=1.5078}
          <at 1 0.5>
         ]}
        YCurve {keys [
          {arriveTangent=2.3356, interpMode=RCIM_Cubic, time=-0.0034, value=1}
          {arriveTangent=5.9046, interpMode=RCIM_Cubic, leaveTangent=-6.5667, tangentMode=RCTM_Break, time=0.2412, value=1.7614}
          {arriveTangent=-1.48, interpMode=RCIM_Cubic, leaveTangent=-1.48, tangentMode=RCTM_User, time=0.3329, value=1.3319}
          <at 1 1>
         ]}
        ZCurve {keys [
          {arriveTangent=1.8377, interpMode=RCIM_Cubic, leaveTangent=1.8377, tangentMode=RCTM_User, time=-0.0042, value=1}
          {arriveTangent=5.4767, interpMode=RCIM_Cubic, leaveTangent=-3.8259, tangentMode=RCTM_Break, time=0.244, value=1.7314}
          {arriveTangent=-1.4717, interpMode=RCIM_Cubic, leaveTangent=-1.4717, tangentMode=RCTM_User, time=0.3368, value=1.3453}
          <at 1 1>
         ]}
       }})
    ))
   &AccelerationForce(Acceleration=&FN_FrontEndVector3Modifier(Gameplay=<0 0 -980>, Front End=<0 0 0>))
   &AddVelocity(Velocity=&FN_FrontEndVector3Modifier(
     Gameplay=<0 0 0>
     Front End=&Multiply_Vector(
      A=<0.3 -50 25>
      B=&VectorFromCurve(VectorCurve=DataInterface() {propertyValues {
         XCurve {keys [<at 0 1>,<at 1 1>]}
         YCurve {keys [<at 0 1>,<at 1 1>]}
         ZCurve {keys [<at 0 0.25 RCIM_Cubic>, {arriveTangent=3.5645, interpMode=RCIM_Cubic, leaveTangent=3.5645, tangentMode=RCTM_User, time=0.65, value=2}]}
        }})
     )
    )) {module=AddVelocity001}
   &SolveForcesAndVelocity(Clamp Velocity=false, Limit Acceleration=false)
   /SetVariables(Particles.Position=&FN_FrontEndVector3Modifier(
     Gameplay=&Lerp_Vector(
      A=&TransformPosition(Position=$Particles.LocalPosition, Source Space=::Local, Destination Space=::Simulation)
      B=Linked() {linkedVariable {name=Particles.Position, type=NiagaraPosition}}
      Alpha=&FloatFromCurve(FloatCurve=DataInterface() {propertyValues {Curve {keys [
           {interpMode=RCIM_Cubic, leaveTangent=0.117, tangentMode=RCTM_Break, time=0.25, value=0}
           {arriveTangent=3.8697, interpMode=RCIM_Cubic, tangentMode=RCTM_Break, time=0.5, value=1}
          ]}}})
     )
     Front End=Linked() {linkedVariable {name=Particles.Position, type=NiagaraPosition}}
    )) {isSetParameters=true}
   &ScaleColor(
    Scale Alpha=&FloatFromCurve(FloatCurve=DataInterface() {propertyValues {Curve {keys [<at 0 1 RCIM_Cubic>,<at 1 0>]}}})
    ScaleA=true
    ScaleRGB=true
   ) {enabled=false}
   &DynamicMaterialParameters(
    Index 0 Param 1=0.8
    Index 0 Param 2=&FloatFromCurve(FloatCurve=DataInterface() {propertyValues {Curve {keys [{arriveTangent=-1.8884, interpMode=RCIM_Cubic, leaveTangent=-1.8884, tangentMode=RCTM_User, time=0, value=1}, <at 1 0.26>]}}})
    Index 0 Param 3=1
    Index 0 Param 4=1
    Module.Param0WriteEnabled=true
    Module.Param1WriteEnabled=true
    Module.Param2WriteEnabled=true
    Module.Param3WriteEnabled=true
   )
   &self:CameraOffsetAngle() {module=CameraOffsetAngle}
   &CameraOffset(Camera Offset Amount=&Multiply_Float(A=3, B=&ClampFloat(Float=$Particles.ZDiffValue, Min=-1, Max=5)))
   &GenerateLocationEvent(
    Event Send Rate=1
    EventCondition:NiagaraBool=&SetBoolByFloatComparison(A=$Particles.NormalizedAge, B=0)
    Float to Send as Random Float:NiagaraFloat=$Particles.SocketFloat
    Module.Use Event Send Rate=true
   )
   &Color(Color=&ColorFromCurve(
     ColorCurve=DataInterface() {propertyValues {
       AlphaCurve {keys [<at 0 1 RCIM_Cubic>]}
       BlueCurve {keys [<at -1 0>,<at -0.4 0>,<at 0.4 0.0546>,<at 1 1 RCIM_Cubic>]}
       GreenCurve {keys [<at -1 0>,<at -0.4 0.9106>,<at 0.4 1>,<at 1 0.0111 RCIM_Cubic>]}
       RedCurve {keys [<at -1 1>,<at -0.4 1>,<at 0.4 0>,<at 1 0 RCIM_Cubic>]}
      }}
     CurveIndex=&Multiply_Float(A=$Particles.ZDiffValue, B=0.01)
    ))
  ]
  rendererClasses [&NiagaraSpriteRendererProperties, &NiagaraMeshRendererProperties]
  renderers [
   {rendererClass=&NiagaraSpriteRendererProperties}
   {rendererClass=&NiagaraMeshRendererProperties, rendererIndex=1}
  ]
 }
}
summary {
 systemName=NS_Glider_ChickenWarrior_Slurp_FX
 emitters [
  {bEnabled=false, emitterName=Drip_A_Legs, rendererClasses [&NiagaraSpriteRendererProperties]}
  {bEnabled=false, emitterName=Drip_B, rendererClasses [&NiagaraSpriteRendererProperties]}
  {emitterName=Drip_Wing, rendererClasses [&NiagaraSpriteRendererProperties]}
  {emitterName=SyrupDrops, rendererClasses [&NiagaraSpriteRendererProperties, &NiagaraMeshRendererProperties]}
 ]
 userVariables [
  User.bIsFrontEnd {type=NiagaraBool}
  User.bIsFrontEndPreview {type=NiagaraBool}
  User.IsSkydiving {type=NiagaraBool}
  User.Land {type=NiagaraFloat, defaultValue=0, description=""}
  User.RunningSpeed {type=NiagaraFloat, defaultValue=0, description=""}
  User.SkeletalMesh {
   type=NiagaraDataInterfaceSkeletalMesh
   defaultValue {
    struct=NiagaraExt_VariableValue_DataInterface
    value {dataInterface=&self:NiagaraDataInterfaceSkeletalMesh_6}
   }
  }
  User.VariantColor {type=LinearColor, description="", defaultValue {struct=LinearColor, value=<rgba 1 1 1 1>}}
 ]
}
