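# Niagara ribbon contrail for the Chicken Warrior male glider, attached to bones or sockets of the user-supplied skeletal mesh.
# A first-loop burst fills partitioned ribbons, each sampling its own bone.
# Rather than trailing by motion, particles are placed every update along a Hermite spline whose end bends with smoothed User.RightDot.
# Panning material color, opacity and width vary along each ribbon, with integrator-driven erosion and time-of-day alpha scaling.
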NiagaraSystem /BRCosmetics/Gadgets/Assets/VinderTech_GliderChute/Glider_ChickenWarrior_Male/FX/NS_Glider_ChickenWarrior_Trail
dependencies {
 usedDataInterfaces [&NiagaraDataInterfaceColorCurve, &NiagaraDataInterfaceCurve]
 usedDynamicInputs [&Add_Float]
 usedModules [
  &SystemState
  &FN_Integrator
  &self:SystemSpawnScript.SystemScriptSource.SystemScriptGraph.NiagaraNodeAssignment_0.SetVariables_#1
 ]
}
emitters {SlurpContrail {
  emitterUpdateScript [
   &EmitterState()
   /Game/Effects/Niagara/Modules/Partition/FN_PartitionAttributes(Particles Per Partition=50, Partitions=4)
   &SpawnBurst_Instantaneous(
    Spawn Count=&MakeCustomIntFromBool(Bool=HlslExpression() {hlslExpression="Emitter.LoopCount == 0"}, True INT=$Emitter.FN_PartitionAttributes.Spawn Total, False INT=0)
    Use Spawn Probability=false
   )
  ]
  particleSpawnScript [
   /Game/Effects/Niagara/Modules/Partition/FN_Particle_partition_v5/FN_Particle_Partition_v5(Enable Ribbon ID=true)
   /SetVariables(Particles.NormalizedID:NiagaraFloat=&Divide_Float(
     A=&MakeFloatFromInt(Int=&ReturnParticleIDAttributes())
     B=&MakeFloatFromInt(Int=$Emitter.ParticlesPerPartition)
    )) {isSetParameters=true}
   /SetVariables(Particles.Partition.ID:NiagaraInt32=&ModuloInt(A=&MakeIntFromFloat(Float=$Particles.NormalizedID), B=$Emitter.Number of Bone Sets)) {isSetParameters=true}
   &InitializeRibbon(
    Module.Write Color=true
    Module.Write Lifetime=true
    Module.Write Mass=true
    Module.Write Material Random=true
    Module.Write Position=false
    Module.Write RibbonTwist=false
    Module.Write RibbonWidth=true
    Ribbon Width=75
   )
   /SetVariables(Particles.InitialRightDot:NiagaraFloat=$User.RightDot) {isSetParameters=true}
  ]
  particleUpdateScript [
   &ParticleState(Kill Particles When Lifetime Has Elapsed=false, Let Infinitely Lived Particles Die When Emitter Deactivates=true)
   &SkeletalMeshLocation(
    Apply Sampled Orientation to Mesh Particles=true
    Bone / Socket Index=$Particles.Partition.ID
    Bone Sampling Mode=::NewEnumerator4
    Inherit Velocity=false
    Interpolate Mesh Sampling=false
    MaskSpawnGroup=false
    Sampled Position Offset=<0 -10 0>
    Sampled Position Offset Coordinate Space:ENiagaraMeshTransforms=::NewEnumerator3
    Skeletal Mesh=$User.SkeletalMesh
    Write Sampled Mesh Attrributes to Particle Attributes=true
   )
   /Game/Effects/Niagara/Modules/Curves/FN_HermiteCubicSpline_v1/FN_HermiteCubicSpline(
    Direction Start:Vector3f=<0 600 0>
    Position Blend:NiagaraFloat=&Multiply_Float(A=$Particles.Initial.Partition.NormalizedID, B=$Particles.Initial.Partition.NormalizedID)
    Position End:Vector3f=&Add_Vector(
     A=$Output.SkeletalMeshLocation.SampledPosition
     B=&Lerp_Vector(
      A=<0 300 0>
      B=<-500 300 0>
      Alpha=&self:InterpolateFloat(Initial Value:NiagaraFloat=$Particles.Initial.InitialRightDot, Rate Of Change:NiagaraFloat=2, Target Value:NiagaraFloat=$User.RightDot)
     )
    )
    Position Start:Vector3f=$Output.SkeletalMeshLocation.SampledPosition
    Write Particles Position=true
   )
   &FN_RibbonMaterial(
    Color=&ColorFromCurve(
     ColorCurve=DataInterface() {propertyValues {
       AlphaCurve {keys [<at -0.0023 1>,<at 1.0323 0 RCIM_Cubic>]}
       BlueCurve {keys [<at 0 2.5 RCIM_Cubic>,<at 0.5 0.6 RCIM_Cubic>]}
       GreenCurve {keys [<at 0 2.5 RCIM_Cubic>,<at 0.5 0.6 RCIM_Cubic>]}
       RedCurve {keys [<at 0 2.5 RCIM_Cubic>,<at 0.5 0.6 RCIM_Cubic>]}
      }}
     CurveIndex=$Particles.Partition.NormalizedID
    )
    Color Pan U Speed=2.5
    Color Scale U=0.5
    Erosion Power=0
    Erossion Scale U=0
    Macro Distortion Amount=0.05
    Macro Pan Speed=1
    Macro Scale U=2
    Opacity Erosion:NiagaraFloat=&FloatFromCurve(
     FloatCurve=DataInterface() {propertyValues {Curve {keys [{interpMode=RCIM_Cubic, tangentMode=RCTM_User, time=0.1, value=1}, <at 1 0>]}}}
     CurveIndex=$System.FN_Integrator002.Value
    )
    Opacity Pan U Offset:NiagaraFloat=&RandomRangeFloat(Minimum=0, Maximum=100, Randomness Mode=::NewEnumerator1, Random Seed=$Particles.Partition.ID, Fixed Random Seed=true)
    Opacity Pan U Speed=2.5
    Opacity Power:NiagaraFloat=&FloatFromCurve(FloatCurve=DataInterface() {propertyValues {Curve {keys [<at 0.25 0.75>,<at 1 1.5>]}}}, CurveIndex=$Particles.Partition.NormalizedID)
    Opacity Scale U=0.5
    Scale Alpha:NiagaraFloat=&FloatFromCurve(
     FloatCurve=DataInterface() {propertyValues {Curve {keys [<at 0 0.45>,<at 0.4952 0.65>,<at 1 0.45>]}}}
     CurveIndex=Linked() {linkedVariable {name=NPC.FortniteNPC.TimeOfDay_0-1, type=NiagaraFloat}}
    )
   ) {module=FN_RibbonMaterial001}
   &ScaleRibbonWidth(Ribbon Width Scale=&FloatFromCurve(
     FloatCurve=DataInterface() {propertyValues {Curve {keys [
         {interpMode=RCIM_Cubic, tangentMode=RCTM_User, time=0, value=0.4}
         {arriveTangent=0.6795, interpMode=RCIM_Cubic, leaveTangent=0.6795, tangentMode=RCTM_User, time=0.5, value=1.18}
         <at 1 0.9>
        ]}}}
     CurveIndex=$Particles.Partition.NormalizedID
    )) {module=ScaleRibbonWidth001}
  ]
  rendererClasses [&NiagaraRibbonRendererProperties]
  renderers [{rendererClass=&NiagaraRibbonRendererProperties}]
 }}
summary {
 systemName=NS_Glider_ChickenWarrior_Trail
 emitters [{emitterName=SlurpContrail, rendererClasses [&NiagaraRibbonRendererProperties]}]
 userVariables [
  User.bIsFrontEnd {type=NiagaraBool}
  User.bIsFrontEndPreview {type=NiagaraBool}
  User.bIsFullyDeployed {type=NiagaraBool, defaultValue=true}
  User.ForwardDot {type=NiagaraFloat, defaultValue=1, description=""}
  User.RibbonIDOffset {type=NiagaraInt32}
  User.RightDot {type=NiagaraFloat, defaultValue=0, description=""}
  User.SkeletalMesh {
   type=NiagaraDataInterfaceSkeletalMesh
   defaultValue {
    struct=NiagaraExt_VariableValue_DataInterface
    value {dataInterface=&self:NiagaraDataInterfaceSkeletalMesh_2}
   }
  }
 ]
}
